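# Physics asset for the arm ribbon mesh on the male medium Desert Shadow player cosmetic body, defining its bodies and joints for physics-driven motion.
# Five small capsule bodies form two two-segment chains (01-02 and 03-04) that branch from a shared base bone.
# All four parent-child constraints share the same angular limits, allowing moderate swing with very little twist.
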PhysicsAsset /BRCosmetics/Characters/Player/Male/Medium/Bodies/M_MED_Desert_Shadow/Meshes/M_MED_Desert_Shadow_Arm_Ribbon_Physics
bodies {
 dyn_arm_ribbon_01 [{shapeType=Capsule, length=1, radius=1.5, shapeName=dyn_arm_ribbon_01_capsule, center=<2 0 0>, rotation=<rot 90 0 0>}]
 dyn_arm_ribbon_02 [{shapeType=Capsule, length=1, radius=1.5, shapeName=dyn_arm_ribbon_02_capsule, rotation=<rot 90 0 0>}]
 dyn_arm_ribbon_03 [{shapeType=Capsule, length=1, radius=1.75, shapeName=dyn_arm_ribbon_03_capsule, center=<2 0 0>, rotation=<rot 90 0 0>}]
 dyn_arm_ribbon_04 [{shapeType=Capsule, length=1, radius=1.75, shapeName=dyn_arm_ribbon_04_capsule, rotation=<rot 90 0 0>}]
 dyn_arm_ribbon_base [{shapeType=Capsule, length=1, radius=1, shapeName=dyn_arm_ribbon_base_capsule, rotation=<rot 90 0 0>}]
}
constraints [
 {bone1Name=dyn_arm_ribbon_01, bone2Name=dyn_arm_ribbon_base, swing1LimitDegrees=25, swing2LimitDegrees=25, twistLimitDegrees=5}
 {bone1Name=dyn_arm_ribbon_02, bone2Name=dyn_arm_ribbon_01, swing1LimitDegrees=25, swing2LimitDegrees=25, twistLimitDegrees=5}
 {bone1Name=dyn_arm_ribbon_03, bone2Name=dyn_arm_ribbon_base, swing1LimitDegrees=25, swing2LimitDegrees=25, twistLimitDegrees=5}
 {bone1Name=dyn_arm_ribbon_04, bone2Name=dyn_arm_ribbon_03, swing1LimitDegrees=25, swing2LimitDegrees=25, twistLimitDegrees=5}
]
