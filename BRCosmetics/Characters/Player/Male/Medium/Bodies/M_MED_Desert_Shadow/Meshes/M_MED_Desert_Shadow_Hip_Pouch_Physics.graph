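# Physics asset for the left and right dynamic hip pouches on the Desert Shadow medium male body cosmetic.
# Each side is a base, 01, 02 chain of capsule bodies hanging from a small pelvis capsule, with fatter capsules on the left pouch.
# Base joints rotate freely on the pelvis.
# Joints further down lock swing1 and carry 15-degree swing2 and twist limits, restricting how the pouches bend.
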PhysicsAsset /BRCosmetics/Characters/Player/Male/Medium/Bodies/M_MED_Desert_Shadow/Meshes/M_MED_Desert_Shadow_Hip_Pouch_Physics
bodies {
 dyn_pouch_hip_01_l [{shapeType=Capsule, length=1.25, radius=3.75, shapeName=dyn_pouch_hip_01_l_capsule, center=<2.5896 0 0>, rotation=<rot 90 0 0>}]
 dyn_pouch_hip_01_r [{shapeType=Capsule, length=1.5, radius=1.5, shapeName=dyn_pouch_hip_01_r_capsule, center=<1 0 0>, rotation=<rot 90 0 0>}]
 dyn_pouch_hip_02_l [{shapeType=Capsule, length=1.25, radius=3.75, shapeName=dyn_pouch_hip_02_l_capsule, center=<2.5896 0 0>, rotation=<rot 90 0 0>}]
 dyn_pouch_hip_02_r [{shapeType=Capsule, length=1.5, radius=1.5, shapeName=dyn_pouch_hip_02_r_capsule, center=<1 0 0>, rotation=<rot 90 0 0>}]
 dyn_pouch_hip_base_l [{shapeType=Capsule, length=1, radius=1, shapeName=dyn_pouch_hip_base_l_capsule}]
 dyn_pouch_hip_base_r [{shapeType=Capsule, length=1, radius=1, shapeName=dyn_pouch_hip_base_r_capsule}]
 pelvis [{shapeType=Capsule, length=1, radius=1, shapeName=pelvis_capsule, rotation=<rot 0 90 0>}]
}
constraints [
 {
  bone1Name=dyn_pouch_hip_base_l
  bone2Name=pelvis
  swing1LimitDegrees=45
  swing1Motion=Free
  swing2LimitDegrees=45
  swing2Motion=Free
  twistLimitDegrees=45
  twistMotion=Free
 }
 {
  bone1Name=dyn_pouch_hip_01_l
  bone2Name=dyn_pouch_hip_base_l
  swing1LimitDegrees=15
  swing1Motion=Locked
  swing2LimitDegrees=15
  twistLimitDegrees=15
 }
 {
  bone1Name=dyn_pouch_hip_02_l
  bone2Name=dyn_pouch_hip_01_l
  swing1LimitDegrees=15
  swing1Motion=Locked
  swing2LimitDegrees=15
  twistLimitDegrees=15
 }
 {
  bone1Name=dyn_pouch_hip_base_r
  bone2Name=pelvis
  swing1LimitDegrees=45
  swing1Motion=Free
  swing2LimitDegrees=45
  swing2Motion=Free
  twistLimitDegrees=45
  twistMotion=Free
 }
 {
  bone1Name=dyn_pouch_hip_01_r
  bone2Name=dyn_pouch_hip_base_r
  swing1LimitDegrees=15
  swing1Motion=Locked
  swing2LimitDegrees=15
  twistLimitDegrees=15
 }
 {
  bone1Name=dyn_pouch_hip_02_r
  bone2Name=dyn_pouch_hip_01_r
  swing1LimitDegrees=15
  swing1Motion=Locked
  swing2LimitDegrees=15
  twistLimitDegrees=15
 }
]
