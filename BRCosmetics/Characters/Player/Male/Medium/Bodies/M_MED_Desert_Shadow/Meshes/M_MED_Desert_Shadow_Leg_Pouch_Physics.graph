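# Physics asset for the thigh pouch on the M_MED_Desert_Shadow male player body cosmetic, covering its dyn_thigh_pouch bones.
# Three capsule bodies form a chain: a small base, then pouch segments 01 and 02, each constrained to the one before it.
# Both constraints set a modest limit angle on swing 1 and very tight ones on swing 2 and twist, favoring hinge-like sway on one axis.
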PhysicsAsset /BRCosmetics/Characters/Player/Male/Medium/Bodies/M_MED_Desert_Shadow/Meshes/M_MED_Desert_Shadow_Leg_Pouch_Physics
bodies {
 dyn_thigh_pouch_01 [{shapeType=Capsule, length=3, radius=3, shapeName=dyn_thigh_pouch_01_capsule, center=<3 0 0>, rotation=<rot 90 0 0>}]
 dyn_thigh_pouch_02 [{shapeType=Capsule, length=3, radius=3, shapeName=dyn_thigh_pouch_base_capsule, center=<3 0 0>, rotation=<rot 90 0 0>}]
 dyn_thigh_pouch_base [{shapeType=Capsule, length=1, radius=1, shapeName=dyn_thigh_pouch_base_capsule}]
}
constraints [
 {bone1Name=dyn_thigh_pouch_01, bone2Name=dyn_thigh_pouch_base, swing1LimitDegrees=15, swing2LimitDegrees=1, twistLimitDegrees=1}
 {bone1Name=dyn_thigh_pouch_02, bone2Name=dyn_thigh_pouch_01, swing1LimitDegrees=15, swing2LimitDegrees=1, twistLimitDegrees=1}
]
