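# Physics asset for the right leg ribbons of the male medium Desert Shadow body cosmetic.
# Small capsule bodies on the right calf and four ribbon bones form two separate two-segment chains, 01 to 02 and 03 to 04, each rooted at the calf.
# All joints share the same angular limits, with moderate swing and very tight twist, restricting how far the ribbons bend and roll.
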PhysicsAsset /BRCosmetics/Characters/Player/Male/Medium/Bodies/M_MED_Desert_Shadow/Meshes/M_MED_Desert_Shadow_Leg_Ribbon_Right_Physics
bodies {
 calf_r [{shapeType=Capsule, length=1, radius=1, shapeName=calf_r_capsule}]
 leg_ribbon_01_r [{shapeType=Capsule, length=2, radius=1, shapeName=leg_ribbon_01_r_capsule, center=<2 0 0>, rotation=<rot 0 90 90>}]
 leg_ribbon_02_r [{shapeType=Capsule, length=2, radius=1, shapeName=leg_ribbon_02_r_capsule, center=<2 0 0>, rotation=<rot 0 90 90>}]
 leg_ribbon_03_r [{shapeType=Capsule, length=2, radius=1, shapeName=leg_ribbon_03_r_capsule, center=<2 0 0>, rotation=<rot 0 90 90>}]
 leg_ribbon_04_r [{shapeType=Capsule, length=2, radius=1, shapeName=leg_ribbon_04_r_capsule, center=<2 0 0>, rotation=<rot 0 90 90>}]
}
constraints [
 {bone1Name=leg_ribbon_02_r, bone2Name=leg_ribbon_01_r, swing1LimitDegrees=20, swing2LimitDegrees=30, twistLimitDegrees=5}
 {bone1Name=leg_ribbon_04_r, bone2Name=leg_ribbon_03_r, swing1LimitDegrees=20, swing2LimitDegrees=30, twistLimitDegrees=5}
 {bone1Name=leg_ribbon_01_r, bone2Name=calf_r, swing1LimitDegrees=20, swing2LimitDegrees=30, twistLimitDegrees=5}
 {bone1Name=leg_ribbon_03_r, bone2Name=calf_r, swing1LimitDegrees=20, swing2LimitDegrees=30, twistLimitDegrees=5}
]
