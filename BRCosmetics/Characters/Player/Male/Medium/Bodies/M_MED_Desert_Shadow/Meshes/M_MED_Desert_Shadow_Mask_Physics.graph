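# Physics asset for the Desert Shadow medium male body's mask mesh, providing capsule bodies for dangling mask, ear and collar bones.
# A four-segment mask chain hangs from dyn_mask_base, with three-segment left and right ear chains branching off its first segment.
# Two-segment collar chains attach to spine_05, as do the mask base, neck and clavicles.
# Chain joints have tight swing and twist limits, while mask base, neck and clavicle joints rotate freely.
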PhysicsAsset /BRCosmetics/Characters/Player/Male/Medium/Bodies/M_MED_Desert_Shadow/Meshes/M_MED_Desert_Shadow_Mask_Physics
bodies {
 clavicle_l [{shapeType=Capsule, length=9.6778, radius=6.5743, shapeName=clavicle_l_capsule, center=<10.556 -3.5633 1.2569>, rotation=<rot 90 0 0>}]
 clavicle_r [{shapeType=Capsule, length=9.722, radius=6.857, shapeName=clavicle_r_capsule, center=<-9.2896 3.6816 -0.3044>, rotation=<rot 90 0 0>}]
 dyn_collar_01_l [{shapeType=Capsule, length=2, radius=2, shapeName=dyn_collar_01_l_capsule, center=<3 0 0>, rotation=<rot 90 0 0>}]
 dyn_collar_01_r [{shapeType=Capsule, length=2, radius=2, shapeName=dyn_collar_01_r_capsule, center=<3 0 0>, rotation=<rot 90 0 0>}]
 dyn_collar_02_l [{shapeType=Capsule, length=2, radius=2, shapeName=dyn_collar_02_l_capsule, center=<3 0 0>, rotation=<rot 90 0 0>}]
 dyn_collar_02_r [{shapeType=Capsule, length=2, radius=2, shapeName=dyn_collar_02_r_capsule, center=<3 0 0>, rotation=<rot 90 0 0>}]
 dyn_ear_01_l [{shapeType=Capsule, length=3, radius=2, shapeName=dyn_ear_01_l_capsule, center=<2 0 0>, rotation=<rot 90 0 0>}]
 dyn_ear_01_r [{shapeType=Capsule, length=3.5, radius=1.5, shapeName=dyn_ear_01_r_capsule, center=<2 0 0>, rotation=<rot 90 0 0>}]
 dyn_ear_02_l [{shapeType=Capsule, length=3, radius=2, shapeName=dyn_ear_02_l_capsule, center=<2 0 0>, rotation=<rot 90 0 0>}]
 dyn_ear_02_r [{shapeType=Capsule, length=3.5, radius=1.5, shapeName=dyn_ear_02_r_capsule, center=<2 0 0>, rotation=<rot 90 0 0>}]
 dyn_ear_03_l [{shapeType=Capsule, length=3, radius=2, shapeName=dyn_ear_03_l_capsule, center=<2 0 0>, rotation=<rot 90 0 0>}]
 dyn_ear_03_r [{shapeType=Capsule, length=3.75, radius=1.25, shapeName=dyn_ear_03_r_capsule, center=<2 0 0>, rotation=<rot 90 0 0>}]
 dyn_mask_01 [{shapeType=Capsule, length=3, radius=1.5, shapeName=dyn_mask_01_capsule, center=<2 -1 0>, rotation=<rot 90 0 0>}]
 dyn_mask_02 [{shapeType=Capsule, length=3, radius=1.5, shapeName=dyn_mask_02_capsule, center=<2 -1 0>, rotation=<rot 90 0 0>}]
 dyn_mask_03 [{shapeType=Capsule, length=3, radius=1.5, shapeName=dyn_mask_03_capsule, center=<2 -1 0>, rotation=<rot 90 0 0>}]
 dyn_mask_04 [{shapeType=Capsule, length=3, radius=1.5, shapeName=dyn_mask_04_capsule, center=<2 -1 0>, rotation=<rot 90 0 0>}]
 dyn_mask_base [{shapeType=Capsule, length=3, radius=3, shapeName=dyn_mask_base_capsule}]
 neck_01 [{shapeType=Capsule, length=1.25, radius=7.8248, shapeName=spine_05_capsule}]
 spine_05 [{shapeType=Capsule, length=17.633, radius=11.5748, shapeName=spine_05_capsule, center=<2.2557 0 0>}]
}
constraints [
 {bone1Name=dyn_mask_01, bone2Name=dyn_mask_base, swing1LimitDegrees=30, swing2LimitDegrees=30, twistLimitDegrees=5}
 {bone1Name=dyn_mask_02, bone2Name=dyn_mask_01, swing1LimitDegrees=30, swing2LimitDegrees=30, twistLimitDegrees=5}
 {bone1Name=dyn_mask_03, bone2Name=dyn_mask_02, swing1LimitDegrees=30, swing2LimitDegrees=30, twistLimitDegrees=5}
 {bone1Name=dyn_mask_04, bone2Name=dyn_mask_03, swing1LimitDegrees=30, swing2LimitDegrees=30, twistLimitDegrees=5}
 {bone1Name=dyn_ear_01_r, bone2Name=dyn_mask_01, swing1LimitDegrees=30, swing2LimitDegrees=30, twistLimitDegrees=3}
 {bone1Name=dyn_ear_02_r, bone2Name=dyn_ear_01_r, swing1LimitDegrees=30, swing2LimitDegrees=30, twistLimitDegrees=3}
 {bone1Name=dyn_ear_03_r, bone2Name=dyn_ear_02_r, swing1LimitDegrees=30, swing2LimitDegrees=30, twistLimitDegrees=3}
 {bone1Name=dyn_ear_01_l, bone2Name=dyn_mask_01, swing1LimitDegrees=30, swing2LimitDegrees=30, twistLimitDegrees=3}
 {bone1Name=dyn_ear_02_l, bone2Name=dyn_ear_01_l, swing1LimitDegrees=30, swing2LimitDegrees=30, twistLimitDegrees=3}
 {bone1Name=dyn_ear_03_l, bone2Name=dyn_ear_02_l, swing1LimitDegrees=30, swing2LimitDegrees=30, twistLimitDegrees=3}
 {
  bone1Name=dyn_mask_base
  bone2Name=spine_05
  swing1LimitDegrees=45
  swing1Motion=Free
  swing2LimitDegrees=45
  swing2Motion=Free
  twistLimitDegrees=45
  twistMotion=Free
 }
 {
  bone1Name=neck_01
  bone2Name=spine_05
  swing1LimitDegrees=45
  swing1Motion=Free
  swing2LimitDegrees=45
  swing2Motion=Free
  twistLimitDegrees=45
  twistMotion=Free
 }
 {bone1Name=dyn_collar_01_l, bone2Name=spine_05, swing1LimitDegrees=15, swing2LimitDegrees=15, twistLimitDegrees=15}
 {bone1Name=dyn_collar_02_l, bone2Name=dyn_collar_01_l, swing1LimitDegrees=15, swing2LimitDegrees=15, twistLimitDegrees=15}
 {bone1Name=dyn_collar_01_r, bone2Name=spine_05, swing1LimitDegrees=15, swing2LimitDegrees=15, twistLimitDegrees=15}
 {bone1Name=dyn_collar_02_r, bone2Name=dyn_collar_01_r, swing1LimitDegrees=15, swing2LimitDegrees=15, twistLimitDegrees=15}
 {
  bone1Name=clavicle_r
  bone2Name=spine_05
  swing1LimitDegrees=45
  swing1Motion=Free
  swing2LimitDegrees=45
  swing2Motion=Free
  twistLimitDegrees=45
  twistMotion=Free
 }
 {
  bone1Name=clavicle_l
  bone2Name=spine_05
  swing1LimitDegrees=45
  swing1Motion=Free
  swing2LimitDegrees=45
  swing2Motion=Free
  twistLimitDegrees=45
  twistMotion=Free
 }
]
