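# Animation Blueprint for the Desert Shadow medium male body cosmetic, adding deformation and secondary motion to the attached parent's copied pose.
# The copied pose feeds a deform Control Rig, then five chained base-bone-space RigidBody sims for arm ribbon, mask, leg pouch, leg ribbon and hip pouch.
# Head and left upper-arm deltas relative to spine_05, computed each update, drive a mask-top rotation before the sims and left-collar scaling after them.
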AnimBlueprint /BRCosmetics/Characters/Player/Male/Medium/Bodies/M_MED_Desert_Shadow/Meshes/M_MED_Desert_Shadow_AnimBP
target_skeleton=&/M_MED_Desert_Shadow_Skeleton
graphs [
 AnimGraph {nodes [
   n1 = Control Rig(Source=Copy Pose From Mesh() {bUseAttachedParent=true}) {
    alphaBoolBlend {blendOption=Linear}
    controlRigAssetReference {blueprintRigClass=&/M_MED_DeformRig_Desert_Shadow_ControlRig_new.M_MED_DeformRig_Desert_Shadow_ControlRig_new_C}
   }
   n2 = "Transform (Modify) Bone - Bone: dyn_mask_top"(
    ComponentPose=Local To Component(LocalPose=@n1)
    Alpha=Get head() -> (head:double)
    Translation="0,0,0"
    Rotation="0.000,-20,0.000"
    Scale="1,1,1"
   ) {
    rotationMode=BMM_Additive
    rotationSpace=BCS_BoneSpace
    translationMode=BMM_Additive
    alphaBoolBlend {blendOption=Linear}
    alphaScaleBiasClamp {bInterpResult=true, interpSpeedIncreasing=15}
    boneToModify {boneName=dyn_mask_top}
    rotation {yaw=-20}
   }
   n3 = RigidBody(ComponentPose=@n2) {
    bOverrideWorldGravity=true
    overridePhysicsAsset=&/M_MED_Desert_Shadow_Arm_Ribbon_Physics
    simulationSpace=BaseBoneSpace
    alphaBoolBlend {blendOption=Linear}
    baseBoneRef {boneName=upperarm_r}
    componentAppliedLinearAccClamp=<2000 2000 2000>
    componentLinearAccScale=<0.3 0.3 0.35>
    componentLinearVelScale=<0.3 0.3 0.35>
    overrideWorldGravity {z=-980}
   }
   n4 = RigidBody(ComponentPose=@n3) {
    bOverrideWorldGravity=true
    overridePhysicsAsset=&/M_MED_Desert_Shadow_Mask_Physics
    simulationSpace=BaseBoneSpace
    alphaBoolBlend {blendOption=Linear}
    baseBoneRef {boneName=spine_05}
    componentAppliedLinearAccClamp=<2000 2000 2000>
    componentLinearAccScale=<0.3 0.3 0.35>
    componentLinearVelScale=<0.3 0.3 0.35>
    overrideWorldGravity {z=-980}
   }
   n5 = RigidBody(ComponentPose=@n4) {
    bOverrideWorldGravity=true
    overridePhysicsAsset=&/M_MED_Desert_Shadow_Leg_Pouch_Physics
    simulationSpace=BaseBoneSpace
    alphaBoolBlend {blendOption=Linear}
    baseBoneRef {boneName=thigh_l}
    componentAppliedLinearAccClamp=<2000 2000 2000>
    componentLinearAccScale=<0.3 0.3 0.35>
    componentLinearVelScale=<0.3 0.3 0.35>
    overrideWorldGravity {z=-980}
   }
   n6 = RigidBody(ComponentPose=@n5) {
    bOverrideWorldGravity=true
    overridePhysicsAsset=&/M_MED_Desert_Shadow_Leg_Ribbon_Right_Physics
    simulationSpace=BaseBoneSpace
    alphaBoolBlend {blendOption=Linear}
    baseBoneRef {boneName=calf_r}
    componentAppliedLinearAccClamp=<2000 2000 2000>
    componentLinearAccScale=<0.3 0.3 0.35>
    componentLinearVelScale=<0.3 0.3 0.35>
    overrideWorldGravity {z=-980}
   }
   n7 = RigidBody(ComponentPose=@n6) {
    bOverrideWorldGravity=true
    overridePhysicsAsset=&/M_MED_Desert_Shadow_Hip_Pouch_Physics
    simulationSpace=BaseBoneSpace
    alphaBoolBlend {blendOption=Linear}
    baseBoneRef {boneName=pelvis}
    componentAppliedLinearAccClamp=<1500 1500 1500>
    componentLinearAccScale=<0.15 0.15 0.2>
    componentLinearVelScale=<0.15 0.15 0.2>
    overrideWorldGravity {z=-980}
   }
   n8 = "Transform (Modify) Bone - Bone: dyn_collar_01_l"(ComponentPose=@n7, Alpha=Get Arm L rot() -> (Arm L rot:double), Translation="0,0,0", Rotation="0,0,0", Scale="0.2,0.2,1.000") {
    rotationMode=BMM_Additive
    scaleMode=BMM_Additive
    translationMode=BMM_Additive
    alphaBoolBlend {blendOption=Linear}
    boneToModify {boneName=dyn_collar_01_l}
    scale=<0.2 0.2>
   }
   Output Pose(Result=Component To Local(ComponentPose=@n8)) {layerGroup=None}
  ]}
 EventGraph {nodes [
   Event BlueprintUpdateAnimation()
   n1 = BreakTransform(InTransform=GetDeltaTransformFromRefPose(self=GetOwnerSkeletalMeshComponent(), BoneName=head, BaseName=spine_05))
   > Set head(head:double=MapRangeClamped(Value=BreakRotator(InRot=@n1.Rotation).Yaw, InRangeA=0, InRangeB=-20)) -> (Output_Get:double)
   > Set Arm L rot(Arm L rot:double=GetAlphaFromReferenceTransformDelta(SkeletalMeshComponent=Get Skel_Mesh(), BoneName=upperarm_l, BaseName=spine_05, InputRange=(X=40,Y=90), TwistAxis="1,0,0")) -> (Output_Get:double)
   Event BlueprintInitializeAnimation()
   > Set Skel_Mesh(Skel_Mesh=GetOwnerSkeletalMeshComponent())
  ]}
]
variables [
 Arm L rot {type=number}
 head {type=number}
 Skel_Mesh {type=SkeletalMeshComponent}
]
